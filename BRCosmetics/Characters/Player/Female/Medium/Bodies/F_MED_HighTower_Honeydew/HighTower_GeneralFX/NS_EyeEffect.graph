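# Niagara eye-flash effect for the HighTower character body, tinted by User.EffectColor.
# Three burst emitters each spawn two short-lived particles, pinned every update to alternating skeletal mesh sockets, presumably the eyes, plus a user offset.
# Burst draws a curve-driven flare with a dark layer, Long adds slower-fading glow sprites and a light, and Long001 is light-only.
# User.bisHeadHidden sets each particle's VisibilityTag, likely toggling renderers when the head is hidden.
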NiagaraSystem /BRCosmetics/Characters/Player/Female/Medium/Bodies/F_MED_HighTower_Honeydew/HighTower_GeneralFX/NS_EyeEffect
dependencies {
 usedDataInterfaces [&NiagaraDataInterfaceVector2DCurve, &NiagaraDataInterfaceCurve, &NiagaraDataInterfaceColorCurve]
 usedDynamicInputs []
 usedModules [&SystemState]
}
emitters {
 Burst {
  emitterUpdateScript [
   &EmitterState(Life Cycle Mode=::NewEnumerator1, Loop Behavior=::NewEnumerator1, Loop Duration=1, UseLoopDelay=false)
   &SpawnBurst_Instantaneous(Spawn Count=2, Use Spawn Probability=false)
  ]
  particleSpawnScript [
   /Niagara/Modules/Spawn/Initialization/InitializeParticle(
    Color=&MultiplyLinearColor(Linear Color=$User.EffectColor, Scale Factor=<1 1 1 0.1>)
    Lifetime=0.3
    Module.Write Color=true
    Module.Write Lifetime=true
    Module.Write Mass=true
    Module.Write Material Random=true
    Module.Write Position=false
    Module.Write Scale=false
    Module.Write SpriteRotation=false
    Module.Write SpriteSize=true
    Position=&SimulationPosition()
    Sprite Size=<40 20>
   )
   /SetVariables(Particles.Color2=$User.EffectColor) {isSetParameters=true, module=SetVariables_#1}
  ]
  particleUpdateScript [
   &SkeletalMeshLocation(
    Bone / Socket Index=&ModuloInt(A=/Niagara/DynamicInputs/Execution/ReturnParticleID(), B=2)
    Inherit Velocity=false
    MaskSpawnGroup=false
    Mesh Sampling Type=::NewEnumerator5
    Sampled Position Offset=&Add_Vector(A=<0 1 0>, B=$User.EffectPositionOffset)
    Sampled Position Offset Coordinate Space:ENiagaraMeshTransforms=::NewEnumerator3
    Skeletal Mesh=$User.SkeletalMesh
    Socket Sampling Mode=::NewEnumerator5
   )
   &ParticleState()
   &SolveForcesAndVelocity(Clamp Velocity=false, Limit Acceleration=false)
   /SetVariables(Particles.Size2:Vector2f=&Vector2DFromCurve(
     Vector2Curve=DataInterface() {propertyValues {
       XCurve {keys [
         {interpMode=RCIM_Cubic, tangentMode=RCTM_User, time=0.0193, value=0}
         {arriveTangent=1004.6009, interpMode=RCIM_Cubic, leaveTangent=1004.6009, time=0.1857, value=103.198}
         {arriveTangent=-609.3447, interpMode=RCIM_Cubic, leaveTangent=-609.3456, tangentMode=RCTM_User, time=0.2844, value=266.3065}
         {arriveTangent=-375.4017, interpMode=RCIM_Cubic, leaveTangent=-375.4017, time=0.7777, value=109.1633}
         <at 1 -2.3497 RCIM_Cubic>
        ]}
       YCurve {keys [
         <at 0.022 10 RCIM_Cubic>
         {arriveTangent=-9.3645, interpMode=RCIM_Cubic, leaveTangent=-9.3645, tangentMode=RCTM_User, time=0.2893, value=3}
         <at 1 0 RCIM_Cubic>
        ]}
      }}
     ScaleCurve=<0.8 1>
    )) {isSetParameters=true, module=SetVariables_#2}
   &ScaleSpriteSize(Scale Factor=&Vector2DFromFloat(Value=&FloatFromCurve(
      FloatCurve=DataInterface() {propertyValues {Curve {keys [<at 0.0058 0.0515 RCIM_Cubic>,<at 0.1454 1.2551>,<at 1 1.7503 RCIM_Cubic>]}}}
      Scale Curve=2
     )))
   &ScaleColor(
    Scale Alpha=&FloatFromCurve(FloatCurve=DataInterface() {propertyValues {Curve {keys [
         <at 0.0119 0>
         {arriveTangent=51.4934, time=0.1877, value=0.6206}
         {arriveTangent=0.248, interpMode=RCIM_Cubic, leaveTangent=0.248, time=0.2409, value=1}
         {arriveTangent=-2.5218, interpMode=RCIM_Cubic, leaveTangent=-2.5218, tangentMode=RCTM_User, time=0.3186, value=0.6531}
         <at 1 0>
        ]}}})
    Scale RGB=&VectorFromFloat(Value=&FloatFromCurve(FloatCurve=DataInterface() {propertyValues {Curve {keys [
          <at 0 1 RCIM_Cubic>
          {arriveTangent=6.4903, interpMode=RCIM_Cubic, leaveTangent=6.4903, time=0.1785, value=10.1828}
          {arriveTangent=-12.3949, interpMode=RCIM_Cubic, leaveTangent=-12.3949, time=0.343, value=3.226}
          <at 1 0>
         ]}}}))
    ScaleA=true
    ScaleRGB=true
   )
   /SetVariables(Particles.DarkColor:LinearColor=&ColorFromCurve(ColorCurve=DataInterface() {propertyValues {
       AlphaCurve {keys [
         <at 0.0019 0>
         {arriveTangent=1.6446, interpMode=RCIM_Cubic, leaveTangent=1.6446, time=0.148, value=0.5}
         <at 0.2375 0.5>
         <at 0.3059 0.5>
         <at 0.6783 0 RCIM_Cubic>
        ]}
       BlueCurve {keys [<at 0.016 0.01 RCIM_Cubic>,<at 1 0.1 RCIM_Cubic>]}
       GreenCurve {keys [<at 0.016 0.01 RCIM_Cubic>,<at 1 0.1 RCIM_Cubic>]}
       RedCurve {keys [<at 0.016 0.01 RCIM_Cubic>,<at 1 0.1 RCIM_Cubic>]}
      }})) {isSetParameters=true, module=SetVariables_#3}
   /SetVariables(Particles.DarkSpriteSize:Vector2f=&Vector2DFromFloat(Value=&FloatFromCurve(FloatCurve=DataInterface() {propertyValues {Curve {keys [
          {arriveTangent=-1440.4427, interpMode=RCIM_Cubic, leaveTangent=-1440.4384, tangentMode=RCTM_User, time=0, value=300}
          {arriveTangent=-574.5275, interpMode=RCIM_Cubic, leaveTangent=175.2135, tangentMode=RCTM_Break, time=0.1761, value=54.365}
          <at 0.9886 162.5396 RCIM_Cubic>
         ]}}}))) {isSetParameters=true, module=SetVariables_#4}
   &CameraOffset(Camera Offset Amount=3.5)
   /SetVariables(Particles.VisibilityTag=&MakeCustomIntFromBool(Bool=$User.bisHeadHidden, True INT=1, False INT=0)) {isSetParameters=true, module=SetVariables_#5}
  ]
  rendererClasses [&NiagaraSpriteRendererProperties]
  renderers [
   {rendererClass=&NiagaraSpriteRendererProperties}
   {rendererClass=&NiagaraSpriteRendererProperties, rendererIndex=1}
   {rendererClass=&NiagaraSpriteRendererProperties, rendererIndex=2}
  ]
 }
 Long {
  emitterUpdateScript [
   &EmitterState(Life Cycle Mode=::NewEnumerator1, Loop Behavior=::NewEnumerator1, Loop Duration=1, UseLoopDelay=false)
   &SpawnBurst_Instantaneous(Spawn Count=2, Use Spawn Probability=false)
  ]
  particleSpawnScript [
   &InitializeParticle(
    Color=&MultiplyLinearColor(Linear Color=$User.EffectColor, Scale Factor=<1 1 1 0.2>)
    Lifetime=0.5
    Module.Write Color=true
    Module.Write Lifetime=true
    Module.Write Mass=true
    Module.Write Material Random=true
    Module.Write Position=false
    Module.Write Scale=false
    Module.Write SpriteRotation=false
    Module.Write SpriteSize=true
    Position=&SimulationPosition()
    Sprite Size=<60 20>
   )
   /SetVariables(Particles.Color2=$User.EffectColor) {isSetParameters=true, module=SetVariables_#1}
   /SetVariables(Particles.CameraOffset2=3) {isSetParameters=true, module=SetVariables_#6}
  ]
  particleUpdateScript [
   &SkeletalMeshLocation(
    Bone / Socket Index=&ModuloInt(A=&ReturnParticleID(), B=2)
    Inherit Velocity=false
    MaskSpawnGroup=false
    Mesh Sampling Type=::NewEnumerator5
    Sampled Position Offset=&Add_Vector(A=<0 1 0>, B=$User.EffectPositionOffset)
    Sampled Position Offset Coordinate Space:ENiagaraMeshTransforms=::NewEnumerator3
    Skeletal Mesh=$User.SkeletalMesh
    Socket Sampling Mode=::NewEnumerator5
   )
   &ParticleState()
   &SolveForcesAndVelocity(Clamp Velocity=false, Limit Acceleration=false)
   /SetVariables(Particles.Size2:Vector2f=&Multiply_Vector2(
     A=<50 1>
     B=&Vector2DFromFloat(Value=&UniformRangedFloat(Minimum=1, Maximum=2))
    )) {isSetParameters=true, module=SetVariables_#2}
   &ScaleSpriteSize(Scale Factor=&Vector2DFromFloat(Value=&FloatFromCurve(FloatCurve=DataInterface() {propertyValues {Curve {keys [<at 0 1>,<at 1 0 RCIM_Cubic>]}}}, Scale Curve=2)))
   /SetVariables(Particles.DarkColor:LinearColor=&ColorFromCurve(ColorCurve=DataInterface() {propertyValues {
       AlphaCurve {keys [
         <at 0.0019 0>
         {arriveTangent=1.6446, interpMode=RCIM_Cubic, leaveTangent=1.6446, time=0.148, value=0.5}
         <at 0.2375 0.5>
         <at 0.3059 0.5>
         <at 0.6783 0 RCIM_Cubic>
        ]}
       BlueCurve {keys [<at 0.016 0.01 RCIM_Cubic>,<at 1 0.1 RCIM_Cubic>]}
       GreenCurve {keys [<at 0.016 0.01 RCIM_Cubic>,<at 1 0.1 RCIM_Cubic>]}
       RedCurve {keys [<at 0.016 0.01 RCIM_Cubic>,<at 1 0.1 RCIM_Cubic>]}
      }})) {isSetParameters=true, module=SetVariables_#3}
   /SetVariables(Particles.DarkSpriteSize:Vector2f=<20 20>) {isSetParameters=true, module=SetVariables_#4}
   &CameraOffset(Camera Offset Amount=1.5)
   /SetVariables(Particles.Light:LinearColor=&MultiplyLinearColor(
     Linear Color=$User.EffectColor
     Scale Factor=&MakeVector4_FromFloat(Float=&FloatFromCurve(FloatCurve=DataInterface() {propertyValues {Curve {keys [<at 0 100 RCIM_Cubic>,<at 1 0 RCIM_Cubic>]}}}))
    )) {isSetParameters=true, module=SetVariables_#7}
   /SetVariables(Particles.EyeColor:LinearColor=&MultiplyLinearColor(
     Linear Color=$User.EffectColor
     Scale Factor=&MakeVector4_FromLinearColor(LinearColor=&MultiplyLinearColor(
       Linear Color=<rgba 100 100 100 0.01>
       Scale Factor=&MakeVector4_FromFloat(Float=&FloatFromCurve(FloatCurve=DataInterface() {propertyValues {Curve {keys [<at 0 1 RCIM_Cubic>, {arriveTangent=-1, interpMode=RCIM_Cubic, leaveTangent=-1, time=0.5556, value=0.6357}, <at 1 0>]}}}))
      ))
    )) {isSetParameters=true}
   &ScaleColor(
    Scale Mode=::NewEnumerator1
    Scale RGBA=&Vector4FromCurve(Vector4Curve=DataInterface() {propertyValues {
       AlphaCurve {keys [<at 0 1>,<at 1 0>]}
       BlueCurve {keys [<at 0 1>,<at 1 1>]}
       GreenCurve {keys [<at 0 1>,<at 1 1>]}
       RedCurve {keys [<at 0 1>,<at 1 1>]}
      }})
    ScaleRGBA=true
   )
   /SetVariables(Particles.Color2=&MultiplyLinearColor(
     Linear Color=$User.EffectColor
     Scale Factor=&MakeVector4_FromLinearColor(LinearColor=&ColorFromCurve(ColorCurve=DataInterface() {propertyValues {
         AlphaCurve {keys [<at 0 1 RCIM_Cubic>,<at 1 0 RCIM_Cubic>]}
         BlueCurve {keys [<at 0 1 RCIM_Cubic>,<at 1 1 RCIM_Cubic>]}
         GreenCurve {keys [<at 0 1 RCIM_Cubic>,<at 1 1 RCIM_Cubic>]}
         RedCurve {keys [<at 0 1 RCIM_Cubic>,<at 1 1 RCIM_Cubic>]}
        }}))
    )) {isSetParameters=true}
   /SetVariables(Particles.Size5:Vector2f=<5 5>) {isSetParameters=true}
   /SetVariables(Particles.Color5:LinearColor=&MultiplyLinearColor(Linear Color=$Particles.EyeColor, Scale Factor=<100 100 100 1>)) {isSetParameters=true}
   /SetVariables(Particles.VisibilityTag=&MakeCustomIntFromBool(Bool=$User.bisHeadHidden, True INT=1, False INT=0)) {isSetParameters=true, module=SetVariables_#5}
  ]
  rendererClasses [&NiagaraSpriteRendererProperties, &NiagaraLightRendererProperties]
  renderers [
   {rendererClass=&NiagaraSpriteRendererProperties}
   {rendererClass=&NiagaraSpriteRendererProperties, rendererIndex=1}
   {rendererClass=&NiagaraSpriteRendererProperties, rendererIndex=2}
   {rendererClass=&NiagaraLightRendererProperties, rendererIndex=3}
   {rendererClass=&NiagaraSpriteRendererProperties, rendererIndex=4}
  ]
 }
 Long001 {
  emitterUpdateScript [
   &EmitterState(Life Cycle Mode=::NewEnumerator1, Loop Behavior=::NewEnumerator1, Loop Duration=1, UseLoopDelay=false)
   &SpawnBurst_Instantaneous(Spawn Count=2, Use Spawn Probability=false)
  ]
  particleSpawnScript [
   &InitializeParticle(
    Color=&MultiplyLinearColor(Linear Color=$User.EffectColor, Scale Factor=<1 1 1 0.2>)
    Lifetime=0.5
    Module.Write Color=true
    Module.Write Lifetime=true
    Module.Write Mass=true
    Module.Write Material Random=true
    Module.Write Position=false
    Module.Write Scale=false
    Module.Write SpriteRotation=false
    Module.Write SpriteSize=true
    Position=&SimulationPosition()
    Sprite Size=<60 20>
   )
   /SetVariables(Particles.Color2=$User.EffectColor) {isSetParameters=true, module=SetVariables_#1}
   /SetVariables(Particles.CameraOffset2=3) {isSetParameters=true, module=SetVariables_#6}
  ]
  particleUpdateScript [
   &SkeletalMeshLocation(
    Bone / Socket Index=&ModuloInt(A=&ReturnParticleID(), B=2)
    Inherit Velocity=false
    MaskSpawnGroup=false
    Mesh Sampling Type=::NewEnumerator5
    Sampled Position Offset=&Add_Vector(A=<0 1 0>, B=$User.EffectPositionOffset)
    Sampled Position Offset Coordinate Space:ENiagaraMeshTransforms=::NewEnumerator3
    Skeletal Mesh=$User.SkeletalMesh
    Socket Sampling Mode=::NewEnumerator5
   )
   &ParticleState()
   &SolveForcesAndVelocity(Clamp Velocity=false, Limit Acceleration=false) {enabled=false}
   /SetVariables(Particles.Light:LinearColor=&MultiplyLinearColor(
     Linear Color=$User.EffectColor
     Scale Factor=&MakeVector4_FromFloat(Float=&FloatFromCurve(FloatCurve=DataInterface() {propertyValues {Curve {keys [<at 0 100 RCIM_Cubic>,<at 1 0 RCIM_Cubic>]}}}, Scale Curve=5))
    )) {isSetParameters=true, module=SetVariables_#7}
   /SetVariables(Particles.VisibilityTag=&MakeCustomIntFromBool(Bool=$User.bisHeadHidden, True INT=1, False INT=0)) {isSetParameters=true, module=SetVariables_#5}
  ]
  rendererClasses [&NiagaraLightRendererProperties]
  renderers [{rendererClass=&NiagaraLightRendererProperties, rendererIndex=0}]
 }
}
summary {
 systemName=NS_EyeEffect
 emitters [
  {emitterName=Burst, rendererClasses [&NiagaraSpriteRendererProperties]}
  {emitterName=Long, rendererClasses [&NiagaraSpriteRendererProperties, &NiagaraLightRendererProperties]}
  {emitterName=Long001, rendererClasses [&NiagaraLightRendererProperties]}
 ]
 userVariables [
  User.bisHeadHidden {type=NiagaraBool}
  User.EffectColor {type=LinearColor, description="", defaultValue {struct=LinearColor, value=<rgba 1 0 0.649 1>}}
  User.EffectPositionOffset {type=Vector3f, description="", defaultValue {struct=Vector3f, value=<0 0 0>}}
  User.SkeletalMesh {
   type=NiagaraDataInterfaceSkeletalMesh
   defaultValue {
    struct=NiagaraExt_VariableValue_DataInterface
    value {dataInterface=&self:NiagaraDataInterfaceSkeletalMesh_0}
   }
  }
 ]
}
